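# Material function in the Masks library that turns a scalar Alpha into a soft threshold mask using ThresholdValue and ThresholdRange.
# A static switch, TwoSidedThreshold, picks a band peaking at the value and fading over the range on both sides, or a one-sided ramp reaching full at the value and staying there.
# Despite the "Unclamped" name, both the range and the final output pass through clamps.
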MaterialFunction /Game/Global/Volumetric/UnclampedThresholdWithRange
graph {
 properties {bExposeToLibrary=true, libraryCategoriesText [Masks]}
 nodes [
  n1 = FunctionInput(Preview=MaterialFunctionCall() {materialFunction=&LinearGradient}.UGradient) {inputName=Alpha, inputType=FunctionInput_Scalar, sortPriority=2}
  n2 = FunctionInput() {inputName=ThresholdValue, inputType=FunctionInput_Scalar, previewValue=<0.5 0 0 1>}
  n3 = Abs(None=Subtract(A=@n1, B=@n2))
  n4 = FunctionInput() {inputName=ThresholdRange, inputType=FunctionInput_Scalar, sortPriority=1, previewValue=<0.1 0 0 1>}
  n5 = Clamp(None=@n4)
  n6 = Multiply(A=Multiply(A=Subtract(A=@n3, B=@n5)) {constB=-1}, B=Divide(A=Constant() {r=1}, B=@n5))
  Clamp(None=@n6)
  n7 = Add(A=Subtract(A=@n1, B=@n2))
  n8 = Constant() {r=1}
  n9 = OneMinus(None=Multiply(A=OneMinus(None=@n7), B=Divide(A=@n8, B=@n5)))
  n10 = FunctionInput(Preview=StaticBool()) {bUsePreviewValueAsDefault=true, inputName=TwoSidedThreshold, inputType=FunctionInput_StaticBool, sortPriority=3}
  FunctionOutput(None=Clamp(None=StaticSwitch(True=@n6, False=@n9, Value=@n10)))
 ]
}
